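# Sound cue for the snow footstep foley of the WinterWinds emote.
# A Random node picks among ten equally weighted snow footstep waves and feeds a Modulator at its default settings.
# The cue uses the emote foley sound class and property sheet, shared emote attenuation, and a reduced volume.
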SoundCue /BRCosmetics/Sounds/Emotes/WinterWinds/Emote_Foley_WinterWinds_Footsteps_Cue
root=Modulator_1
nodes [
 {id=Modulator_1, children [Random_24]}
 {
  id=Random_24
  children [
   WavePlayer_160
   WavePlayer_161
   WavePlayer_162
   WavePlayer_163
   WavePlayer_164
   WavePlayer_165
   WavePlayer_166
   WavePlayer_167
   WavePlayer_168
   WavePlayer_169
  ]
  properties {weights [1,1,1,1,1,1,1,1,1,1]}
 }
 {id=WavePlayer_160, properties {soundWaveAssetPtr=/BRCosmetics/Sounds/Emotes/WinterWinds/Footsteps/Emote_WinterWinds_Footstep_Snow_01}}
 {id=WavePlayer_161, properties {soundWaveAssetPtr=/BRCosmetics/Sounds/Emotes/WinterWinds/Footsteps/Emote_WinterWinds_Footstep_Snow_02}}
 {id=WavePlayer_162, properties {soundWaveAssetPtr=/BRCosmetics/Sounds/Emotes/WinterWinds/Footsteps/Emote_WinterWinds_Footstep_Snow_03}}
 {id=WavePlayer_163, properties {soundWaveAssetPtr=/BRCosmetics/Sounds/Emotes/WinterWinds/Footsteps/Emote_WinterWinds_Footstep_Snow_04}}
 {id=WavePlayer_164, properties {soundWaveAssetPtr=/BRCosmetics/Sounds/Emotes/WinterWinds/Footsteps/Emote_WinterWinds_Footstep_Snow_05}}
 {id=WavePlayer_165, properties {soundWaveAssetPtr=/BRCosmetics/Sounds/Emotes/WinterWinds/Footsteps/Emote_WinterWinds_Footstep_Snow_06}}
 {id=WavePlayer_166, properties {soundWaveAssetPtr=/BRCosmetics/Sounds/Emotes/WinterWinds/Footsteps/Emote_WinterWinds_Footstep_Snow_07}}
 {id=WavePlayer_167, properties {soundWaveAssetPtr=/BRCosmetics/Sounds/Emotes/WinterWinds/Footsteps/Emote_WinterWinds_Footstep_Snow_08}}
 {id=WavePlayer_168, properties {soundWaveAssetPtr=/BRCosmetics/Sounds/Emotes/WinterWinds/Footsteps/Emote_WinterWinds_Footstep_Snow_09}}
 {id=WavePlayer_169, properties {soundWaveAssetPtr=/BRCosmetics/Sounds/Emotes/WinterWinds/Footsteps/Emote_WinterWinds_Footstep_Snow_10}}
]
properties {
 attenuationSettings=&Fort_Emotes_Athena_Att
 audioPropertiesSheet=/BRCosmetics/Sounds/Settings/PropertySheets/Emotes/APS_Cosmetic_Emote_Foley
 maxDistance=1600
 soundClassObject=/Game/Sounds/FortSoundClassesAndModes/Fort_Foley_Emote
 volumeMultiplier=0.7
}
